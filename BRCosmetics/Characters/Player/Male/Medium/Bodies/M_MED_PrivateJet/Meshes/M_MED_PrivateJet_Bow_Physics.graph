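# Physics asset for the dynamic bow bones on the medium male PrivateJet player body.
# Left and right master bones hang from spine_05, each leading to a two-bone chain that branches into A, B and C strands, all using capsules.
# Bow joints have tight swing and zero twist limits, except the free left master.
# Head, upper spine, clavicle and upper-arm capsules form the surrounding body, linked by free angular constraints.
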PhysicsAsset /BRCosmetics/Characters/Player/Male/Medium/Bodies/M_MED_PrivateJet/Meshes/M_MED_PrivateJet_Bow_Physics
bodies {
 clavicle_l [{
   shapeType=Capsule
   length=9.6692
   radius=8.2376
   shapeName=clavicle_l_capsule
   center=<4.2415 -1.3611 -1.9988>
   rotation=<rot 85.0371 -9.2849 -5.8374>
  }]
 clavicle_r [{
   shapeType=Capsule
   length=6.6692
   radius=8.2376
   shapeName=clavicle_l_capsule
   center=<-4.2415 1.3611 1.9988>
   rotation=<rot 85.0371 -9.2849 174.1626>
  }]
 dyn_bowA_l_01 [{shapeType=Capsule, length=3.9678, radius=2.0322, shapeName=dyn_bowA_l_01_capsule, center=<3.7611 1 0>, rotation=<rot 90 0 0>}]
 dyn_bowA_l_02 [{shapeType=Capsule, length=2.4678, radius=2.0322, shapeName=dyn_bowA_l_02_capsule, center=<3.7611 1 0>, rotation=<rot 90 0 0>}]
 dyn_bowA_r_01 [{shapeType=Capsule, length=5.4678, radius=1.0322, shapeName=dyn_bowA_r_01_capsule, center=<3.7611 0 0>, rotation=<rot 90 0 0>}]
 dyn_bowA_r_02 [{shapeType=Capsule, length=4.9678, radius=1.0322, shapeName=dyn_bowA_r_02_capsule, center=<3.7611 -1 0>, rotation=<rot 90 0 0>}]
 dyn_bowB_l_01 [{shapeType=Capsule, length=7.9678, radius=2.0322, shapeName=dyn_bowB_l_01_capsule, center=<5.1171 1 0>, rotation=<rot 90 0 0>}]
 dyn_bowB_l_02 [{shapeType=Capsule, length=3.4678, radius=2.5322, shapeName=dyn_bowB_l_02_capsule, center=<3.7611 1 0>, rotation=<rot 90 0 0>}]
 dyn_bowB_r_01 [{shapeType=Capsule, length=5.4678, radius=1.0322, shapeName=dyn_bowB_r_01_capsule, center=<3.7611 0 0>, rotation=<rot 90 0 0>}]
 dyn_bowB_r_02 [{shapeType=Capsule, length=6.9678, radius=1.0322, shapeName=dyn_bowB_r_02_capsule, center=<3.7611 0 0>, rotation=<rot 90 0 0>}]
 dyn_bowC_l_01 [{shapeType=Capsule, length=4.4678, radius=1.5322, shapeName=dyn_bowC_l_01_capsule, center=<3.7611 1 0>, rotation=<rot 90 0 0>}]
 dyn_bowC_r_01 [{shapeType=Capsule, length=5.4678, radius=1.0322, shapeName=dyn_bowC_r_01_capsule, center=<3.7611 0 0>, rotation=<rot 90 0 0>}]
 dyn_bow_l_01 [{shapeType=Capsule, length=1, radius=2.8, shapeName=dyn_bow_l_01_capsule, center=<2.5 2 0>, rotation=<rot 90 -14.0362 -5.2812>}]
 dyn_bow_l_02 [{shapeType=Capsule, length=3.8484, radius=2.8, shapeName=dyn_bow_l_02_capsule, center=<3.2614 2 0>}]
 dyn_bow_l_MSTR [{shapeType=Capsule, length=0.5, radius=0.5, shapeName=dyn_bow_l_MSTR_capsule}]
 dyn_bow_r_01 [{shapeType=Capsule, length=2.2061, radius=2, shapeName=dyn_bow_r_01_capsule, center=<3 -1.9417 -0.0442>, rotation=<rot 86.8881 4.1127 0.0001>}]
 dyn_bow_r_02 [{shapeType=Capsule, length=6.0296, radius=1.0832, shapeName=dyn_bow_r_02_capsule, center=<4.6368 -0.5 0>}]
 dyn_bow_r_MSTR [{shapeType=Capsule, length=0.5, radius=0.5, shapeName=dyn_bow_r_MSTR_capsule}]
 head [{shapeType=Capsule, length=1, radius=11.5, shapeName=head_capsule, center=<9.2401 2.5456 0>, rotation=<rot 0 -90 0>}]
 spine_04 [{shapeType=Capsule, length=10.4451, radius=10.7427, shapeName=spine_04_capsule, center=<6.6248 0 0>, rotation=<rot 0 -90 0>}]
 spine_05 [{shapeType=Capsule, length=9.8749, radius=11.3242, shapeName=spine_05_capsule, center=<-0.9448 -3.4988 0>}]
 upperarm_l [{shapeType=Capsule, length=24.7732, radius=4.3593, shapeName=upperarm_l_capsule, center=<9.3127 0 -1.5215>, rotation=<rot 90 0 0>}]
 upperarm_r [{shapeType=Capsule, length=24.7732, radius=4.3593, shapeName=upperarm_l_capsule, center=<-9.3127 0 1.5215>, rotation=<rot 90 0 180>}]
}
constraints [
 {
  bone1Name=clavicle_l
  bone2Name=spine_05
  swing1LimitDegrees=45
  swing1Motion=Free
  swing2LimitDegrees=45
  swing2Motion=Free
  twistLimitDegrees=45
  twistMotion=Free
 }
 {
  bone1Name=upperarm_l
  bone2Name=clavicle_l
  swing1LimitDegrees=45
  swing1Motion=Free
  swing2LimitDegrees=45
  swing2Motion=Free
  twistLimitDegrees=45
  twistMotion=Free
 }
 {bone1Name=dyn_bowC_r_01, bone2Name=dyn_bow_r_02, swing1LimitDegrees=5, swing2LimitDegrees=3, twistLimitDegrees=0}
 {bone1Name=dyn_bow_l_01, bone2Name=dyn_bow_l_MSTR, swing1LimitDegrees=16, swing2LimitDegrees=20, twistLimitDegrees=0}
 {bone1Name=dyn_bow_l_02, bone2Name=dyn_bow_l_01, swing1LimitDegrees=12, swing2LimitDegrees=5, twistLimitDegrees=0}
 {bone1Name=dyn_bowA_l_01, bone2Name=dyn_bow_l_02, swing1LimitDegrees=5, swing2LimitDegrees=5, twistLimitDegrees=0}
 {bone1Name=dyn_bowA_l_02, bone2Name=dyn_bowA_l_01, swing1LimitDegrees=5, swing2LimitDegrees=5, twistLimitDegrees=0}
 {bone1Name=dyn_bowB_l_01, bone2Name=dyn_bow_l_02, swing1LimitDegrees=5, swing2LimitDegrees=5, twistLimitDegrees=0}
 {bone1Name=dyn_bowB_l_02, bone2Name=dyn_bowB_l_01, swing1LimitDegrees=5, swing2LimitDegrees=5, twistLimitDegrees=0}
 {bone1Name=dyn_bowC_l_01, bone2Name=dyn_bow_l_02, swing1LimitDegrees=5, swing2LimitDegrees=5, twistLimitDegrees=0}
 {bone1Name=dyn_bow_r_MSTR, bone2Name=spine_05, swing1LimitDegrees=10, swing2LimitDegrees=0, twistLimitDegrees=0}
 {bone1Name=dyn_bow_r_01, bone2Name=dyn_bow_r_MSTR, swing1LimitDegrees=16, swing2LimitDegrees=2, twistLimitDegrees=0}
 {bone1Name=dyn_bow_r_02, bone2Name=dyn_bow_r_01, swing1LimitDegrees=15, swing2LimitDegrees=0, twistLimitDegrees=0}
 {bone1Name=dyn_bowA_r_01, bone2Name=dyn_bow_r_02, swing1LimitDegrees=5, swing2LimitDegrees=3, twistLimitDegrees=0}
 {bone1Name=dyn_bowA_r_02, bone2Name=dyn_bowA_r_01, swing1LimitDegrees=5, swing2LimitDegrees=5, twistLimitDegrees=0}
 {bone1Name=dyn_bowB_r_01, bone2Name=dyn_bow_r_02, swing1LimitDegrees=5, swing2LimitDegrees=3, twistLimitDegrees=0}
 {bone1Name=dyn_bowB_r_02, bone2Name=dyn_bowB_r_01, swing1LimitDegrees=5, swing2LimitDegrees=5, twistLimitDegrees=0}
 {
  bone1Name=dyn_bow_l_MSTR
  bone2Name=spine_05
  swing1LimitDegrees=45
  swing1Motion=Free
  swing2LimitDegrees=45
  swing2Motion=Free
  twistLimitDegrees=45
  twistMotion=Free
 }
 {
  bone1Name=spine_05
  bone2Name=spine_04
  swing1LimitDegrees=45
  swing1Motion=Free
  swing2LimitDegrees=45
  swing2Motion=Free
  twistLimitDegrees=45
  twistMotion=Free
 }
 {
  bone1Name=upperarm_r
  bone2Name=clavicle_r
  swing1LimitDegrees=45
  swing1Motion=Free
  swing2LimitDegrees=45
  swing2Motion=Free
  twistLimitDegrees=45
  twistMotion=Free
 }
 {
  bone1Name=head
  bone2Name=spine_05
  swing1LimitDegrees=45
  swing1Motion=Free
  swing2LimitDegrees=45
  swing2Motion=Free
  twistLimitDegrees=45
  twistMotion=Free
 }
 {
  bone1Name=clavicle_r
  bone2Name=spine_05
  swing1LimitDegrees=45
  swing1Motion=Free
  swing2LimitDegrees=45
  swing2Motion=Free
  twistLimitDegrees=45
  twistMotion=Free
 }
]
