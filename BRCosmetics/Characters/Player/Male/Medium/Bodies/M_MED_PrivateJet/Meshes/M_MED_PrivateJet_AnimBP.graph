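# Player cosmetic AnimBP for the PrivateJet body, adding secondary motion to coat tails and bows.
# Copied parent pose runs through Control Rig, then coat-tail and bow RigidBody sims with custom gravity.
# The PrivateJet emote montage switches physics assets and scales coat-tail bones, while the emote_BuiltIn curve fades bow physics toward slot animation.
# EventGraph also feeds noise and pelvis/spine velocity to material WPO.
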AnimBlueprint /BRCosmetics/Characters/Player/Male/Medium/Bodies/M_MED_PrivateJet/Meshes/M_MED_PrivateJet_AnimBP
target_skeleton=/Game/Characters/Player/Male/Medium/Bodies/M_MED_PrivateJet/Meshes/M_MED_PrivateJet_Skeleton
graphs [
 AnimGraph {nodes [
   n1 = RigidBody!(
    ComponentPose=Local To Component(LocalPose=Use cached pose 'bowPoseCache'())
    Alpha=Get curveInverseValue() -> (curveInverseValue:double)
    OverrideWorldGravity:Vector=Get Gravity Override Bow() -> (Gravity Override Bow:Vector)
    ExternalForce
   ) -> (Pose) {
    bClampLinearTranslationLimitToRefPose=true
    bOverrideWorldGravity=true
    overridePhysicsAsset=&/M_MED_PrivateJet_Bow_FE_Physics
    simulationSpace=BaseBoneSpace
    alphaBoolBlend {blendOption=Linear}
    alphaScaleBiasClamp {bInterpResult=true, interpSpeedDecreasing=0}
    baseBoneRef {boneName=spine_05}
    componentAppliedLinearAccClamp=<6000 6000 6000>
    componentLinearAccScale {x=0.05, z=0.05}
    componentLinearVelScale {x=0.05, z=0.05}
    overrideWorldGravity {z=-980}
   }
   n2 = RigidBody(
    ComponentPose=Local To Component(LocalPose=Use cached pose 'bowPoseCache'())
    Alpha=Get curveInverseValue()
    OverrideWorldGravity=Get Gravity Override Bow()
   ) {
    bClampLinearTranslationLimitToRefPose=true
    bOverrideWorldGravity=true
    overridePhysicsAsset=&/M_MED_PrivateJet_Bow_Physics
    simulationSpace=BaseBoneSpace
    alphaBoolBlend {blendOption=Linear}
    alphaScaleBiasClamp {bInterpResult=true, interpSpeedDecreasing=0}
    baseBoneRef {boneName=spine_05}
    componentAppliedLinearAccClamp=<6000 6000 6000>
    componentLinearAccScale {x=0.05, z=0.05}
    componentLinearVelScale {x=0.05, z=0.05}
    overrideWorldGravity {z=-980}
   }
   Output Pose(Result=Blend Poses by bool(
     BlendPose_0=Component To Local(ComponentPose=@n1)
     BlendPose_1=Component To Local(ComponentPose=@n2)
     bActiveValue=Get isPlayingPrivateJetEmote() -> (isPlayingPrivateJetEmote:bool)
    ) {blendType=Linear}) {layerGroup=None}
   n3 = Control Rig!(Source=Copy Pose From Mesh() {bUseAttachedParent=true}, Alpha, isTargeting:bool=Get bIsTargeting()) -> (Pose) {
    alphaBoolBlend {blendOption=Linear}
    controlRigAssetReference {blueprintRigClass=&/M_MED_PrivateJet_ControlRig.M_MED_PrivateJet_ControlRig_C}
   }
   Save cached pose 'CRCache'(Pose=@n3) {cacheName=CRCache}
   n4 = RigidBody!(
    ComponentPose=Local To Component(LocalPose=Use cached pose 'CRCache'())
    Alpha
    OverrideWorldGravity=Get Gravity Override Coat Tail() -> (Gravity Override Coat Tail:Vector)
    ComponentLinearAccScale:Vector="0.4,0.4,0.5"
    ComponentLinearVelScale:Vector="0.4,0.4,0.5"
   ) -> (Pose) {
    bClampLinearTranslationLimitToRefPose=true
    bOverrideWorldGravity=true
    overridePhysicsAsset=&/M_MED_PrivateJet_CoatTail_Emote_Physics
    simulationSpace=BaseBoneSpace
    alphaBoolBlend {blendOption=Linear}
    baseBoneRef {boneName=root}
    componentAppliedLinearAccClamp=<3000 3000 3000>
    componentLinearAccScale=<0.45 0.5 0.4>
    componentLinearVelScale=<0.45 0.5 0.4>
    overrideWorldGravity {z=-980}
   }
   n5 = "Transform (Modify) Bone - Bone: dyn_coatTail_l_01"(ComponentPose=@n4, Alpha=1, Translation="0,0,0", Rotation="0,0,0", Scale="0.9,1,1") {
    scaleMode=BMM_Additive
    scaleSpace=BCS_BoneSpace
    alphaBoolBlend {blendOption=Linear}
    boneToModify {boneName=dyn_coatTail_l_01}
    scale {x=0.9}
   }
   n6 = "Transform (Modify) Bone - Bone: dyn_coatTail_r_01"(ComponentPose=@n5, Alpha=1, Translation="0,0,0", Rotation="0,0,0", Scale="0.9,1,1") {
    scaleMode=BMM_Additive
    scaleSpace=BCS_BoneSpace
    alphaBoolBlend {blendOption=Linear}
    boneToModify {boneName=dyn_coatTail_r_01}
    scale {x=0.9}
   }
   n7 = "Transform (Modify) Bone - Bone: dyn_coatTail_l_03"(ComponentPose=@n6, Alpha=1, Translation="0,0,0", Rotation="0,0,0", Scale="0.7,1,1") {
    scaleMode=BMM_Additive
    scaleSpace=BCS_BoneSpace
    alphaBoolBlend {blendOption=Linear}
    boneToModify {boneName=dyn_coatTail_l_03}
    scale {x=0.7}
   }
   n8 = "Transform (Modify) Bone - Bone: dyn_coatTail_r_03"(ComponentPose=@n7, Alpha=1, Translation="0,0,0", Rotation="0,0,0", Scale="0.7,1,1") {
    scaleMode=BMM_Additive
    scaleSpace=BCS_BoneSpace
    alphaBoolBlend {blendOption=Linear}
    boneToModify {boneName=dyn_coatTail_r_03}
    scale {x=0.7}
   }
   n9 = "Transform (Modify) Bone - Bone: dyn_coatTailB_l_02"(ComponentPose=@n8, Alpha=1, Translation="0,0,0", Rotation="0,0,0", Scale="0.7,1,1") {
    scaleMode=BMM_Additive
    scaleSpace=BCS_BoneSpace
    alphaBoolBlend {blendOption=Linear}
    boneToModify {boneName=dyn_coatTailB_l_02}
    scale {x=0.7}
   }
   n10 = "Transform (Modify) Bone - Bone: dyn_coatTailC_l_02"(ComponentPose=@n9, Alpha=1, Translation="0,0,0", Rotation="0,0,0", Scale="0.7,1,1") {
    scaleMode=BMM_Additive
    scaleSpace=BCS_BoneSpace
    alphaBoolBlend {blendOption=Linear}
    boneToModify {boneName=dyn_coatTailC_l_02}
    scale {x=0.7}
   }
   n11 = "Transform (Modify) Bone - Bone: dyn_coatTailB_r_02"(ComponentPose=@n10, Alpha=1, Translation="0,0,0", Rotation="0,0,0", Scale="0.7,1,1") {
    scaleMode=BMM_Additive
    scaleSpace=BCS_BoneSpace
    alphaBoolBlend {blendOption=Linear}
    boneToModify {boneName=dyn_coatTailB_r_02}
    scale {x=0.7}
   }
   n12 = "Transform (Modify) Bone - Bone: dyn_coatTailC_r_02"(ComponentPose=@n11, Alpha=1, Translation="0,0,0", Rotation="0,0,0", Scale="0.7,1,1") {
    scaleMode=BMM_Additive
    scaleSpace=BCS_BoneSpace
    alphaBoolBlend {blendOption=Linear}
    boneToModify {boneName=dyn_coatTailC_r_02}
    scale {x=0.7}
   }
   n13 = RigidBody(
    ComponentPose=Local To Component(LocalPose=Use cached pose 'CRCache'())
    OverrideWorldGravity=Get Gravity Override Coat Tail()
    ComponentLinearAccScale="0.4,0.4,0.5"
    ComponentLinearVelScale="0.4,0.4,0.5"
   ) {
    bClampLinearTranslationLimitToRefPose=true
    bOverrideWorldGravity=true
    overridePhysicsAsset=&/M_MED_PrivateJet_CoatTail_Physics
    simulationSpace=BaseBoneSpace
    alphaBoolBlend {blendOption=Linear}
    baseBoneRef {boneName=root}
    componentAppliedLinearAccClamp=<3000 3000 3000>
    componentLinearAccScale=<0.45 0.5 0.4>
    componentLinearVelScale=<0.45 0.5 0.4>
    overrideWorldGravity {z=-980}
   }
   Save cached pose 'CoatTailCache'(Pose=Blend Poses by bool(
     BlendPose_0=Component To Local(ComponentPose=@n12)
     BlendPose_1=Component To Local(ComponentPose=@n13)
     BlendTime_0=0.8
     bActiveValue=Get isPlayingPrivateJetEmote()
    ) {blendType=Linear}) {cacheName=CoatTailCache}
   Save cached pose 'bowPoseCache'(Pose=Layered blend per bone(
     BasePose=Use cached pose 'CoatTailCache'()
     BlendPoses_0=Slot 'DefaultSlot'()
     BlendWeights_0=Get buildInEmoteDetector() -> (buildInEmoteDetector:double)
    ) {layerSetup [{branchFilters [{boneName=dyn_bow_l_MSTR}, {boneName=dyn_bow_r_MSTR}]}]}) {cacheName=bowPoseCache}
  ]}
 EventGraph {nodes [
   n1 = Event BlueprintUpdateAnimation()
   > Set Delta Time X(Delta Time X=@n1.DeltaTimeX)
   > n2 = Sequence!(execute) -> (then_0, then_1, then_2:Exec, then_3:Exec, then_4:Exec, then_5:Exec)
   > n3 = Gravity Override State Function DEPRECATED(
    CharacterPartAnimInstance=Self-Reference()
    GameTimeInSeconds=Get Delta Time X()
    GravityOverrideParams=Get Gravity Override Params Coat Tail() -> (Gravity Override Params Coat Tail:Gravity Override Params Struct Structure)
   )
   > Set Gravity Override Coat Tail(Gravity Override Coat Tail:Vector=@n3) -> (Output_Get:Vector)
   n4 = Gravity Override State Function DEPRECATED(
    execute=@n2.then_1
    CharacterPartAnimInstance=Self-Reference()
    GameTimeInSeconds=0
    GravityOverrideParams=Get Gravity Override Params Bow() -> (Gravity Override Params Bow:Gravity Override Params Struct Structure)
   )
   > Set Gravity Override Bow(Gravity Override Bow:Vector=@n4) -> (Output_Get:Vector)
   n5 = Branch(execute=@n2.then_2, Condition=Get bIsPlayingEmote())
   > n6 = Branch(Condition=Get bIsInFrontEnd())
   n7 = GetCurveValue(self=GetOwnerFrontEndAnimInstance(), CurveName=emote_BuiltIn)
   > Set buildInEmoteDetector(buildInEmoteDetector:double=@n7) -> (Output_Get:double)
   > Set curveInverseValue(curveInverseValue:double=Remap Float(Input=@n7, OldRange=(X=0,Y=1), NewRange=(X=1,Y=0)) -> (Output=~)) -> (Output_Get:double)
   n8 = GetCurveValue(
    self=GetOwnerFortPlayerAnimInstance!(self:Custom Character Part Anim Instance Object Reference) -> (ReturnValue:Fort Player Anim Instance Object Reference)
    CurveName=emote_BuiltIn
   )
   Set buildInEmoteDetector(execute=@n6.else, buildInEmoteDetector=@n8)
   > Set curveInverseValue(curveInverseValue=Remap Float(Input=@n8, OldRange=(X=0,Y=1), NewRange=(X=1,Y=0)) -> (Output=~))
   Set curveInverseValue(execute=@n5.else, curveInverseValue=1)
   Set WPO Globals Function(
    execute=@n2.then_3
    Material=Get matInstanceDynamic() -> (matInstanceDynamic:Material Instance Dynamic Object Reference)
    Skeletal Mesh=Get skelMesh() -> (skelMesh:Skeletal Mesh Component Object Reference)
    "Local Noise Transform  Location"="Get Local Noise Transform  Location"() -> ("Local Noise Transform  Location":Vector)
    Local Noise Offset=Get Local Noise Offset()
   )
   > Set WPO Regional Function!(
    execute
    Material:Material Instance Dynamic Object Reference=Get matInstanceDynamic()
    Overlay Material:Material Instance Dynamic Object Reference
    Skeletal Mesh:Skeletal Mesh Component Object Reference=Get skelMesh()
    Velocity Transform:Name=pelvis
    Velocity Interp Speed:double=Get Velocity Interp Speed()
    Collision Transform:Name=None
    Collision Transform Axis:EAxis Enum
    Invert Collision Transform Axis:bool=false
    Delta Time X:double=Get Delta Time X()
    Previous Velocity Transform Location:"Vector (by ref)"=Get Previous Velocity Transform Location() -> (Previous Velocity Transform Location:Vector)
    Previous Velocity Transform Velocity:"Vector (by ref)"=Get Previous Velocity Transform Velocity() -> (Previous Velocity Transform Velocity:Vector)
    Region:int=1
   ) -> (then)
   > Set WPO Regional Function(
    Material=Get matInstanceDynamic()
    Skeletal Mesh=Get skelMesh()
    Velocity Transform=spine_05
    Velocity Interp Speed=Get Velocity Interp Speed()
    Collision Transform=None
    Invert Collision Transform Axis=false
    Delta Time X=Get Delta Time X()
    Previous Velocity Transform Location=Get Previous Velocity Transform Location_0() -> (Previous Velocity Transform Location_0:Vector)
    Previous Velocity Transform Velocity=Get Previous Velocity Transform Velocity_0() -> (Previous Velocity Transform Velocity_0:Vector)
    Region=2
   )
   n9 = Branch(execute=@n2.then_4, Condition=Get bIsInFrontEnd())
   n10 = GetAnimInstance(self=GetOwnerSkeletalMeshComponent())
   > Is Valid(InputObject=@n10)
   > Set isPlayingPrivateJetEmote(isPlayingPrivateJetEmote:bool=Montage_IsPlaying(self=@n10, Montage=/BRCosmetics/Animation/Game/MainPlayer/Emotes/PrivateJet/Emote_PrivateJet_CMM_FE_M)) -> (Output_Get:bool)
   n11 = GetAnimInstance(self=GetOwnerSkeletalMeshComponent())
   Is Valid(exec=@n9.else, InputObject=@n11)
   > Set isPlayingPrivateJetEmote(isPlayingPrivateJetEmote=Montage_IsPlaying(self=@n11, Montage=/BRCosmetics/Animation/Game/MainPlayer/Emotes/PrivateJet/Emote_PrivateJet_CMM_M))
   Event OnCharacterCustomizationLoaded()
   > n12 = Sequence!(execute) -> (then_0, then_1)
   > n13 = Set skelMesh(skelMesh:Skeletal Mesh Component Object Reference=GetPartSkeletalMeshComponent()) -> (Output_Get:Skeletal Mesh Component Object Reference)
   > n14 = Get Dynamic Material Instance(SkeletalMeshComponent=@n13)
   > Set matInstanceDynamic(matInstanceDynamic:Material Instance Dynamic Object Reference=@n14) -> (Output_Get:Material Instance Dynamic Object Reference)
   Do Once(execute=@n12.then_1)
   n15 = CheckForBackbling()
   > n16 = Branch(Condition=OR Boolean(A=@n15.bIsBackblingACape, B=@n15.bIsBackblingLong))
   > Set AccVelValues(AccVelValues:Vector="0.1,0.1,0.1") -> (Output_Get:Vector)
   Set AccVelValues(execute=@n16.else, AccVelValues="0.4,0.4,0.5")
  ]}
]
variables [
 AccVelValues {type=Vector}
 buildInEmoteDetector {type=number}
 curveInverseValue {type=number}
 Delta Time X {type=number}
 doTheThing {type=number}
 Gravity Override Bow {type=Vector}
 Gravity Override Coat Tail {type=Vector}
 Gravity Override Params Bow {type=GravityOverrideParamsStruct}
 Gravity Override Params Coat Tail {type=GravityOverrideParamsStruct}
 Is Backbling A Cape {type=boolean}
 isPlayingPrivateJetEmote {type=boolean}
 lapsedLocalTime {type=number}
 Local Noise Offset {type=Vector}
 "Local Noise Transform  Location" {type=Vector}
 matInstanceDynamic {type=MaterialInstanceDynamic}
 NewVar {type=AnimMontage, default=null}
 NewVar_0 {type=AnimMontage, default=null}
 Previous Velocity Transform Location {type=Vector}
 Previous Velocity Transform Location_0 {type=Vector}
 Previous Velocity Transform Velocity {type=Vector}
 Previous Velocity Transform Velocity_0 {type=Vector}
 RBANBowAlpha {type=number}
 skelMesh {type=SkeletalMeshComponent}
 Velocity Interp Speed {type=number, default=20}
]
