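# Physics asset for the PrivateJet medium male player body, providing collision shapes that, by its name, serve as cape colliders.
# Capsules cover the pelvis, spine, clavicles and neck, with a sphere on the head; arm and leg bodies have no shapes.
# Large capsules also sit on left and right coat-tail chains, which hang from spine_01 and fork into A, B and C branches.
# All constraints leave swing and twist free.
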PhysicsAsset /BRCosmetics/Characters/Player/Male/Medium/Bodies/M_MED_PrivateJet/Meshes/M_MED_PrivateJet_CapeColliders
bodies {
 calf_l []
 calf_r []
 clavicle_l [{shapeType=Capsule, length=12.0192, radius=6.8871, center=<4.6291 -1.0512 -0.635>, rotation=<rot 87.274 -18.9662 150.6088>}]
 clavicle_r [{shapeType=Capsule, length=12.0192, radius=6.8871, center=<-4.6291 1.0512 0.635>, rotation=<rot 87.274 -18.9662 -29.3912>}]
 dyn_coatTailA_l_01 [{
   shapeType=Capsule
   length=8.5
   radius=15
   shapeName=dyn_coatTailA_l_01_capsule
   center=<8.1719 0.2933 -10.8082>
   rotation=<rot 0 -87.5729 -89.6068>
  }]
 dyn_coatTailA_l_02 [{shapeType=Capsule, length=15, radius=17, center=<2.9954 0 -10.7607>, rotation=<rot 90 0 0>}]
 dyn_coatTailA_r_01 [{
   shapeType=Capsule
   length=14.5
   radius=10.5
   shapeName=dyn_coatTailA_r_01_capsule
   center=<8.107 0.0061 -5.6183>
   rotation=<rot 0 -90.5411 -88.4383>
  }]
 dyn_coatTailA_r_02 [{shapeType=Capsule, length=15, radius=17, center=<2.9954 0 -10.7607>, rotation=<rot 90 0 0>}]
 dyn_coatTailB_l_01 [{
   shapeType=Capsule
   length=14.5
   radius=10.5
   shapeName=dyn_coatTailB_l_01_capsule
   center=<7.9809 0.182 -5.5948>
   rotation=<rot 0 -82.2821 -88.7583>
  }]
 dyn_coatTailB_l_02 [{shapeType=Capsule, length=15, radius=17, center=<2.9954 0 -10.7607>, rotation=<rot 90 0 0>}]
 dyn_coatTailB_r_01 [{
   shapeType=Capsule
   length=14.5
   radius=10.5
   shapeName=dyn_coatTailB_r_01_capsule
   center=<7.8639 -0.009 -5.5758>
   rotation=<rot 0 -96.3207 -87.9808>
  }]
 dyn_coatTailB_r_02 [{shapeType=Capsule, length=15, radius=17, center=<2.9954 0 -10.7607>, rotation=<rot 90 0 0>}]
 dyn_coatTailC_l_01 [{
   shapeType=Capsule
   length=14.5
   radius=10.5
   shapeName=dyn_coatTailC_l_01_capsule
   center=<7.7877 0.2828 -5.8472>
   rotation=<rot 0 84.3864 -88.6126>
  }]
 dyn_coatTailC_l_02 [{
   shapeType=Capsule
   length=14.5
   radius=10.5
   shapeName=dyn_coatTailC_l_02_capsule
   center=<5.1756 1.0331 -5.7962>
   rotation=<rot 0 -77.8473 -89.5729>
  }]
 dyn_coatTailC_r_01 [{
   shapeType=Capsule
   length=14.5
   radius=10.5
   shapeName=dyn_coatTailC_r_01_capsule
   center=<7.5616 0.0152 -5.5084>
   rotation=<rot 0 98.0292 -89.3928>
  }]
 dyn_coatTailC_r_02 [{
   shapeType=Capsule
   length=14.5
   radius=10.5
   shapeName=dyn_coatTailC_r_02_capsule
   center=<5.1351 0.3029 -5.4404>
   rotation=<rot 0 88.0333 -89.5438>
  }]
 dyn_coatTail_l_01 [{shapeType=Capsule, length=4, radius=16, center=<0.6644 0 -12.969>, rotation=<rot 90 0 0>}]
 dyn_coatTail_l_03 [{shapeType=Capsule, length=6.5, radius=15.5, center=<0.6644 0 -11.7497>, rotation=<rot 90 0 0>}]
 dyn_coatTail_r_01 [{shapeType=Capsule, length=4, radius=16, center=<0.6644 0 -12.969>, rotation=<rot 90 0 0>}]
 dyn_coatTail_r_03 [{shapeType=Capsule, length=6.5, radius=15.5, center=<0.6644 0 -11.7497>, rotation=<rot 90 0 0>}]
 foot_l []
 foot_r []
 head [{shapeType=Sphere, radius=17.3125, center=<11.028 -0.174 0>}]
 lowerarm_l []
 lowerarm_r []
 neck_02 [{shapeType=Capsule, length=13.4375, radius=7.3175, center=<-0.2642 -0.3408 0>, rotation=<rot 90 -171.1193 -175.3978>}]
 pelvis [{shapeType=Capsule, length=8.6456, radius=9.0388, center=<-3.197 0.7838 0>}]
 spine_01 [{shapeType=Capsule, length=9.9438, radius=8.5586, center=<-1.7965 1.1778 0>}]
 spine_02 [
  {shapeType=Capsule, length=11.8264, radius=9.1322, center=<-1.711 1.6789 0.0001>}
  {shapeType=Capsule, length=33.8956, radius=7.0807, center=<2.5024 0.4992 3.9825>, rotation=<rot 90 0.0049 0.0049>}
  {shapeType=Capsule, length=34.375, radius=6.6013, center=<2.4704 0.4991 -3.9825>, rotation=<rot 90 -0.0396 -0.0396>}
 ]
 spine_03 [{shapeType=Capsule, length=14.7107, radius=5.2031, center=<0.1445 0.5544 0.0001>}]
 spine_04 [{shapeType=Capsule, length=15.4404, radius=7.6325, center=<2.766 -1.5254 0.0002>}]
 spine_05 [
  {shapeType=Capsule, length=12.5226, radius=11.195, center=<-3.8272 1.5654 0>}
  {shapeType=Capsule, length=17.9518, radius=9.9425, center=<-0.2916 0.4452 0>}
 ]
 thigh_l []
 thigh_r []
 upperarm_l []
 upperarm_r []
}
constraints [
 {
  bone1Name=spine_01
  bone2Name=pelvis
  swing1LimitDegrees=45
  swing1Motion=Free
  swing2LimitDegrees=45
  swing2Motion=Free
  twistLimitDegrees=45
  twistMotion=Free
 }
 {
  bone1Name=spine_02
  bone2Name=spine_01
  swing1LimitDegrees=45
  swing1Motion=Free
  swing2LimitDegrees=45
  swing2Motion=Free
  twistLimitDegrees=45
  twistMotion=Free
 }
 {
  bone1Name=spine_03
  bone2Name=spine_02
  swing1LimitDegrees=45
  swing1Motion=Free
  swing2LimitDegrees=45
  swing2Motion=Free
  twistLimitDegrees=45
  twistMotion=Free
 }
 {
  bone1Name=spine_04
  bone2Name=spine_03
  swing1LimitDegrees=45
  swing1Motion=Free
  swing2LimitDegrees=45
  swing2Motion=Free
  twistLimitDegrees=45
  twistMotion=Free
 }
 {
  bone1Name=spine_05
  bone2Name=spine_04
  swing1LimitDegrees=45
  swing1Motion=Free
  swing2LimitDegrees=45
  swing2Motion=Free
  twistLimitDegrees=45
  twistMotion=Free
 }
 {
  bone1Name=neck_02
  bone2Name=spine_05
  swing1LimitDegrees=45
  swing1Motion=Free
  swing2LimitDegrees=45
  swing2Motion=Free
  twistLimitDegrees=45
  twistMotion=Free
 }
 {
  bone1Name=head
  bone2Name=neck_02
  swing1LimitDegrees=45
  swing1Motion=Free
  swing2LimitDegrees=45
  swing2Motion=Free
  twistLimitDegrees=45
  twistMotion=Free
 }
 {
  bone1Name=thigh_l
  bone2Name=pelvis
  swing1LimitDegrees=45
  swing1Motion=Free
  swing2LimitDegrees=45
  swing2Motion=Free
  twistLimitDegrees=45
  twistMotion=Free
 }
 {
  bone1Name=calf_l
  bone2Name=thigh_l
  swing1LimitDegrees=45
  swing1Motion=Free
  swing2LimitDegrees=45
  swing2Motion=Free
  twistLimitDegrees=45
  twistMotion=Free
 }
 {
  bone1Name=foot_l
  bone2Name=calf_l
  swing1LimitDegrees=45
  swing1Motion=Free
  swing2LimitDegrees=45
  swing2Motion=Free
  twistLimitDegrees=45
  twistMotion=Free
 }
 {
  bone1Name=upperarm_r
  bone2Name=clavicle_r
  swing1LimitDegrees=45
  swing1Motion=Free
  swing2LimitDegrees=45
  swing2Motion=Free
  twistLimitDegrees=45
  twistMotion=Free
 }
 {
  bone1Name=lowerarm_r
  bone2Name=upperarm_r
  swing1LimitDegrees=45
  swing1Motion=Free
  swing2LimitDegrees=45
  swing2Motion=Free
  twistLimitDegrees=45
  twistMotion=Free
 }
 {
  bone1Name=lowerarm_l
  bone2Name=upperarm_l
  swing1LimitDegrees=45
  swing1Motion=Free
  swing2LimitDegrees=45
  swing2Motion=Free
  twistLimitDegrees=45
  twistMotion=Free
 }
 {
  bone1Name=upperarm_l
  bone2Name=clavicle_l
  swing1LimitDegrees=45
  swing1Motion=Free
  swing2LimitDegrees=45
  swing2Motion=Free
  twistLimitDegrees=45
  twistMotion=Free
 }
 {
  bone1Name=clavicle_l
  bone2Name=spine_05
  swing1LimitDegrees=45
  swing1Motion=Free
  swing2LimitDegrees=45
  swing2Motion=Free
  twistLimitDegrees=45
  twistMotion=Free
 }
 {
  bone1Name=thigh_r
  bone2Name=pelvis
  swing1LimitDegrees=45
  swing1Motion=Free
  swing2LimitDegrees=45
  swing2Motion=Free
  twistLimitDegrees=45
  twistMotion=Free
 }
 {
  bone1Name=foot_r
  bone2Name=calf_r
  swing1LimitDegrees=45
  swing1Motion=Free
  swing2LimitDegrees=45
  swing2Motion=Free
  twistLimitDegrees=45
  twistMotion=Free
 }
 {
  bone1Name=clavicle_r
  bone2Name=spine_05
  swing1LimitDegrees=45
  swing1Motion=Free
  swing2LimitDegrees=45
  swing2Motion=Free
  twistLimitDegrees=45
  twistMotion=Free
 }
 {
  bone1Name=calf_r
  bone2Name=thigh_r
  swing1LimitDegrees=45
  swing1Motion=Free
  swing2LimitDegrees=45
  swing2Motion=Free
  twistLimitDegrees=45
  twistMotion=Free
 }
 {
  bone1Name=dyn_coatTailB_l_02
  bone2Name=dyn_coatTailB_l_01
  swing1LimitDegrees=45
  swing1Motion=Free
  swing2LimitDegrees=45
  swing2Motion=Free
  twistLimitDegrees=45
  twistMotion=Free
 }
 {
  bone1Name=dyn_coatTailA_l_02
  bone2Name=dyn_coatTailA_l_01
  swing1LimitDegrees=45
  swing1Motion=Free
  swing2LimitDegrees=45
  swing2Motion=Free
  twistLimitDegrees=45
  twistMotion=Free
 }
 {
  bone1Name=dyn_coatTail_l_03
  bone2Name=dyn_coatTail_l_01
  swing1LimitDegrees=45
  swing1Motion=Free
  swing2LimitDegrees=45
  swing2Motion=Free
  twistLimitDegrees=45
  twistMotion=Free
 }
 {
  bone1Name=dyn_coatTail_l_01
  bone2Name=spine_01
  swing1LimitDegrees=45
  swing1Motion=Free
  swing2LimitDegrees=45
  swing2Motion=Free
  twistLimitDegrees=45
  twistMotion=Free
 }
 {
  bone1Name=dyn_coatTail_r_01
  bone2Name=spine_01
  swing1LimitDegrees=45
  swing1Motion=Free
  swing2LimitDegrees=45
  swing2Motion=Free
  twistLimitDegrees=45
  twistMotion=Free
 }
 {
  bone1Name=dyn_coatTail_r_03
  bone2Name=dyn_coatTail_r_01
  swing1LimitDegrees=45
  swing1Motion=Free
  swing2LimitDegrees=45
  swing2Motion=Free
  twistLimitDegrees=45
  twistMotion=Free
 }
 {
  bone1Name=dyn_coatTailA_r_02
  bone2Name=dyn_coatTailA_r_01
  swing1LimitDegrees=45
  swing1Motion=Free
  swing2LimitDegrees=45
  swing2Motion=Free
  twistLimitDegrees=45
  twistMotion=Free
 }
 {
  bone1Name=dyn_coatTailB_r_02
  bone2Name=dyn_coatTailB_r_01
  swing1LimitDegrees=45
  swing1Motion=Free
  swing2LimitDegrees=45
  swing2Motion=Free
  twistLimitDegrees=45
  twistMotion=Free
 }
 {
  bone1Name=dyn_coatTailA_l_01
  bone2Name=dyn_coatTail_l_03
  swing1LimitDegrees=45
  swing1Motion=Free
  swing2LimitDegrees=45
  swing2Motion=Free
  twistLimitDegrees=45
  twistMotion=Free
 }
 {
  bone1Name=dyn_coatTailB_l_01
  bone2Name=dyn_coatTail_l_03
  swing1LimitDegrees=45
  swing1Motion=Free
  swing2LimitDegrees=45
  swing2Motion=Free
  twistLimitDegrees=45
  twistMotion=Free
 }
 {
  bone1Name=dyn_coatTailC_l_01
  bone2Name=dyn_coatTail_l_03
  swing1LimitDegrees=45
  swing1Motion=Free
  swing2LimitDegrees=45
  swing2Motion=Free
  twistLimitDegrees=45
  twistMotion=Free
 }
 {
  bone1Name=dyn_coatTailC_l_02
  bone2Name=dyn_coatTailC_l_01
  swing1LimitDegrees=45
  swing1Motion=Free
  swing2LimitDegrees=45
  swing2Motion=Free
  twistLimitDegrees=45
  twistMotion=Free
 }
 {
  bone1Name=dyn_coatTailA_r_01
  bone2Name=dyn_coatTail_r_03
  swing1LimitDegrees=45
  swing1Motion=Free
  swing2LimitDegrees=45
  swing2Motion=Free
  twistLimitDegrees=45
  twistMotion=Free
 }
 {
  bone1Name=dyn_coatTailB_r_01
  bone2Name=dyn_coatTail_r_03
  swing1LimitDegrees=45
  swing1Motion=Free
  swing2LimitDegrees=45
  swing2Motion=Free
  twistLimitDegrees=45
  twistMotion=Free
 }
 {
  bone1Name=dyn_coatTailC_r_01
  bone2Name=dyn_coatTail_r_03
  swing1LimitDegrees=45
  swing1Motion=Free
  swing2LimitDegrees=45
  swing2Motion=Free
  twistLimitDegrees=45
  twistMotion=Free
 }
 {
  bone1Name=dyn_coatTailC_r_02
  bone2Name=dyn_coatTailC_r_01
  swing1LimitDegrees=45
  swing1Motion=Free
  swing2LimitDegrees=45
  swing2Motion=Free
  twistLimitDegrees=45
  twistMotion=Free
 }
]
